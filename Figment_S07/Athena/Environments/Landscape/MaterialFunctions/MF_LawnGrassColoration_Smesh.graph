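# Material function that recolors grass material attributes into golf lawn, with striped fairway and green tints, plus a simpler LowMobile fairway output.
# Base color red drives a brightness range times grass color, blended in by the fairway mask, height-lerped with inverted opacity for darkened edges on the main path.
# A lawn mask texture on landscape coordinates supplies stripe variation, also output as low-end tints.
# The green mask is constant zero, leaving the green blend inactive.
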MaterialFunction /Figment_S07/Athena/Environments/Landscape/MaterialFunctions/MF_LawnGrassColoration_Smesh
graph {nodes [
  n1 = FunctionInput() {bUsePreviewValueAsDefault=true, inputName=In_LowMobile_Material, inputType=FunctionInput_MaterialAttributes}
  n2 = ScalarParameter() {defaultValue=0.075, parameterName=Golf_Grass_Min}
  n3 = ScalarParameter() {defaultValue=1, parameterName=Golf_Grass_Max}
  n4 = MaterialFunctionCall!(Material=@n1) {materialFunction=/Engine/Functions/MaterialLayerFunctions/MatLayerBlend_BreakBaseColor}
  n5 = VectorParameter() {parameterName=Golf_GrassColor, defaultValue=<rgba 0.5 1.05 0.2516 1>}
  n6 = Multiply(A=LinearInterpolate(A=@n2, B=@n3, Alpha=ComponentMask(None=@n4.BaseColor) {r=true}), B=@n5.RGB)
  n7 = FunctionInput() {bUsePreviewValueAsDefault=true, inputName=Golf_Fairway_Mask, inputType=FunctionInput_Scalar}
  n8 = TextureSample(UVs=Multiply(A=LandscapeLayerCoords() {mappingScale=12}, B=ScalarParameter() {defaultValue=0.5, parameterName=GolfStripes_Tiling})) {
   samplerSource=SSM_Wrap_WorldGroupSettings
   samplerType=SAMPLERTYPE_LinearColor
   texture=/Figment_S07/Athena/Environments/Landscape/Textures/T_Athena_LawnMask_01_M
  }
  n9 = LinearInterpolate(
   A=ScalarParameter() {defaultValue=0.9, parameterName=Fairway_Color_Min}
   B=ScalarParameter() {defaultValue=1.1, parameterName=Fairway_Color_Max}
   Alpha=@n8.B
  )
  n10 = MaterialFunctionCall(
   In=Multiply(A=Power(Base=@n7), B=ScalarParameter() {defaultValue=0.75, parameterName=Golf_EdgeTransition})
   Contrast=ScalarParameter() {defaultValue=5, parameterName=Golf_Fairway_Contrast}
  ) {materialFunction=&CheapContrast}
  FunctionOutput(None=MaterialFunctionCall!(
    Material=MaterialFunctionCall(Material=@n1, NewBaseColor=@n6, Mask=@n7) {materialFunction=/Engine/Functions/MaterialLayerFunctions/MatLayerBlend_OverrideBaseColor}.Blended Material
    NewBaseColor=Multiply(A=@n6, B=@n9)
    Mask=@n10.Result
   ) {materialFunction=&MatLayerBlend_OverrideBaseColor}.Blended Material) {outputName=LowMobile_Fairway}
  n11 = FunctionInput() {bUsePreviewValueAsDefault=true, inputType=FunctionInput_MaterialAttributes}
  n12 = MaterialFunctionCall(
   Transition Phase=@n7
   Height Texture=OneMinus(None=MaterialFunctionCall!(Mat attributes=@n11) {materialFunction=/Engine/Functions/Engine_MaterialFunctions02/MatLayerBlend_BreakOpacity}.Result)
   Contrast=ScalarParameter() {defaultValue=1, parameterName=Golf_EdgeContrast}
  ) {materialFunction=&HeightLerp}
  n13 = ScalarParameter() {defaultValue=0.5, parameterName=EdgeShadowTest}
  n14 = MaterialFunctionCall(Material=@n11) {materialFunction=&MatLayerBlend_BreakBaseColor}
  n15 = Multiply(A=LinearInterpolate(A=@n2, B=@n3, Alpha=ComponentMask(None=@n14.BaseColor) {r=true}), B=@n5.RGB)
  n16 = MaterialFunctionCall(Material=@n11, NewBaseColor=Multiply(A=Power(Base=@n12.Alpha, Exp=@n13), B=@n15), Mask=@n12.Alpha) {materialFunction=&MatLayerBlend_OverrideBaseColor}
  n17 = Multiply(A=@n15, B=@n9)
  n18 = LinearInterpolate(
   A=ScalarParameter() {defaultValue=1.2, parameterName=Green_Color_Min}
   B=ScalarParameter() {defaultValue=1.25, parameterName=Green_Color_Max}
   Alpha=@n8.G
  )
  n19 = Constant()
  FunctionOutput(None=MaterialFunctionCall(
    Material=MaterialFunctionCall(Material=@n16.Blended Material, NewBaseColor=@n17, Mask=@n10.Result) {materialFunction=&MatLayerBlend_OverrideBaseColor}.Blended Material
    NewBaseColor=Multiply(A=@n15, B=@n18)
    Mask=@n19
   ) {materialFunction=&MatLayerBlend_OverrideBaseColor}.Blended Material)
  FunctionOutput(None=Multiply(A=@n9, B=ScalarParameter() {defaultValue=1.25, parameterName=Golf_LowEnd_FairwayTint})) {outputName=Fairway_Color}
  FunctionOutput(None=Multiply(A=@n18, B=ScalarParameter() {defaultValue=1.1, parameterName=Golf_LowEnd_GreenTint})) {outputName=Golf_Green_Color}
  FunctionOutput(None=@n19) {outputName=Green_Mask}
  MaterialFunctionCall() {materialFunction=&MatLayerBlend_BreakOpacity}
  ScalarParameter() {defaultValue=1.3, parameterName=Golf_EdgeBrightness}
 ]}
